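# Physics asset for the Strawberry Pilot female medium body's shoes and collar mesh, setting up secondary motion on dyn_ collar and shoe bones.
# Front collar capsules hinge on small base bodies with narrow limits; shoe tongues attach to calf twists and the left-only loop to the foot, with slightly wider limits.
# Neck, clavicle, upper-arm and calf shapes provide body volumes around them, with upper-arm-to-clavicle angular motion left free.
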PhysicsAsset /BRCosmetics/Characters/Player/Female/Medium/Bodies/F_MED_Strawberry_Pilot/Meshes/F_MED_StrawberryPilot_ShoesCollar_PhysicsAsset
bodies {
 calf_twist_01_l [
  {shapeType=Capsule, length=6.6932, radius=4.5057, center=<4.9758 -8.0255 0.5002>, rotation=<rot 0 0 110.0391>}
  {shapeType=Box, extentX=2.8251, extentY=7.5, extentZ=17.1749, center=<2.3529 -7.5352 0.0001>, rotation=<rot 0 0 180>}
 ]
 calf_twist_01_r [{shapeType=Box, extentX=2.8251, extentY=10.3439, extentZ=19.4784, center=<0.3423 3.2594 0>, rotation=<rot 0 21.2333 0>}]
 clavicle_l [{shapeType=Sphere, radius=4.994, center=<10.2574 -2.3294 -2.4259>}]
 clavicle_r [{shapeType=Sphere, radius=5.1106, center=<-5.9107 2.7294 2.1188>}]
 dyn_collarFrontBase_l [{shapeType=Capsule, length=0.505, radius=0.505, rotation=<rot 0 -90 0>}]
 dyn_collarFrontBase_r [{shapeType=Capsule, length=0.505, radius=0.505, rotation=<rot 0 -90 0>}]
 dyn_collarFront_l [{shapeType=Capsule, length=7.8869, radius=4.1918, center=<-6.4964 -0.825 -0.9524>, rotation=<rot 0 98.0631 -62.3236>}]
 dyn_collarFront_r [{shapeType=Capsule, length=7.9833, radius=4.1232, center=<6.6075 0.8632 0.8173>, rotation=<rot 0 97.5389 -59.5302>}]
 dyn_shoeLoop_l [{shapeType=Capsule, length=5.6054, radius=4.1975, center=<-4.7336 2.5849 0.4666>, rotation=<rot 90 0 0>}]
 dyn_shoeTongue_l [{shapeType=Capsule, length=5.9368, radius=3.7248, center=<-3.7365 -0.1613 -0.6982>, rotation=<rot 90 0 0>}]
 dyn_shoeTongue_r [{shapeType=Capsule, length=5.9368, radius=3.7248, center=<5.4616 -0.1615 -0.6982>, rotation=<rot 90 0 0>}]
 foot_l [{shapeType=Capsule, length=1, radius=1, rotation=<rot 0 0 -84.04>}]
 neck_01 [{shapeType=Sphere, radius=6.755}]
 upperarm_l [{shapeType=Capsule, length=12.108, radius=4.1145, center=<4.3053 -0.4836 -3.0216>, rotation=<rot 90 0 -179.9999>}]
 upperarm_r [{shapeType=Capsule, length=12.108, radius=4.1145, center=<-4.3053 0.4836 3.0216>, rotation=<rot 90 0 0.0001>}]
}
constraints [
 {bone1Name=dyn_collarFront_l, bone2Name=dyn_collarFrontBase_l, swing1LimitDegrees=15, swing2LimitDegrees=1, twistLimitDegrees=1}
 {bone1Name=dyn_collarFront_r, bone2Name=dyn_collarFrontBase_r, swing1LimitDegrees=15, swing2LimitDegrees=1, twistLimitDegrees=1}
 {bone1Name=dyn_shoeTongue_r, bone2Name=calf_twist_01_r, swing1LimitDegrees=20, swing2LimitDegrees=5, twistLimitDegrees=5}
 {bone1Name=dyn_shoeLoop_l, bone2Name=foot_l, swing1LimitDegrees=20, swing2LimitDegrees=5, twistLimitDegrees=5}
 {bone1Name=dyn_shoeTongue_l, bone2Name=calf_twist_01_l, swing1LimitDegrees=20, swing2LimitDegrees=5, twistLimitDegrees=5}
 {
  bone1Name=upperarm_l
  bone2Name=clavicle_l
  swing1LimitDegrees=45
  swing1Motion=Free
  swing2LimitDegrees=45
  swing2Motion=Free
  twistLimitDegrees=45
  twistMotion=Free
 }
 {
  bone1Name=upperarm_r
  bone2Name=clavicle_r
  swing1LimitDegrees=45
  swing1Motion=Free
  swing2LimitDegrees=45
  swing2Motion=Free
  twistLimitDegrees=45
  twistMotion=Free
 }
]
