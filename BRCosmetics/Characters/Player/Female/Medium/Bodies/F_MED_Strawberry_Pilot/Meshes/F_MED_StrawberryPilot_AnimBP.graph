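# Secondary-motion AnimBlueprint for the female Strawberry Pilot body cosmetic, layered on a pose copied from the parent mesh.
# Swing-alpha bone corrections, constraints, AnimDynamics and two RigidBody solvers move collar, jacket, zippers and frills, with DBNO overrides, into a cached pose.
# That pose gets crouch-dependent flap and frill offsets, then Pose Drivers correct both shoulders.
# The event graph resets dynamics after initialization and toggles an upward jacket force around jumps, with delays.
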AnimBlueprint /BRCosmetics/Characters/Player/Female/Medium/Bodies/F_MED_Strawberry_Pilot/Meshes/F_MED_StrawberryPilot_AnimBP
target_skeleton=&/F_MED_StrawberryPilot_Skeleton
graphs [
 AnimGraph {nodes [
   n1 = "Transform (Modify) Bone - Bone: Flap_Front_Driven_Bone"(
    ComponentPose=Local To Component(LocalPose=Use cached pose 'PreCrouch'())
    Alpha=Get CrouchSpeedAlpha() -> (CrouchSpeedAlpha:float)
    Translation="0,0,0"
    Rotation="35.000,26.356,-30"
    Scale="1,1,1"
   ) {
    rotationMode=BMM_Additive
    rotationSpace=BCS_BoneSpace
    translationMode=BMM_Additive
    translationSpace=BCS_BoneSpace
    alphaBoolBlend {blendOption=Linear}
    alphaScaleBiasClamp {bInterpResult=true}
    boneToModify {boneName=Flap_Front_Driven_Bone}
    rotation=<rot 35 26.356 -30>
   }
   n2 = "Transform (Modify) Bone - Bone: dyn_shoeLoop_r"(ComponentPose=@n1, Alpha=1, Translation="0,0,0", Rotation="0,35,0", Scale="1,1,1") {
    rotationMode=BMM_Additive
    rotationSpace=BCS_BoneSpace
    alphaBoolBlend {blendOption=Linear}
    boneToModify {boneName=dyn_shoeLoop_r}
    rotation {yaw=35}
   }
   n3 = "Transform (Modify) Bone - Bone: driver_frill_frt_l"(
    ComponentPose=Local To Component(LocalPose=Use cached pose 'PreCrouch'())
    Alpha=1
    Translation="0,0,0"
    Rotation="0.000,-15,0.000"
    Scale="1,1,1"
   ) {
    rotationMode=BMM_Additive
    rotationSpace=BCS_BoneSpace
    alphaBoolBlend {blendOption=Linear}
    boneToModify {boneName=driver_frill_frt_l}
    rotation {yaw=-15}
   }
   n4 = "Transform (Modify) Bone - Bone: driver_frill_frt_r"(ComponentPose=@n3, Alpha=1, Translation="0,0,0", Rotation="0.000,-24.999,0.000", Scale="1,1,1") {
    rotationMode=BMM_Additive
    rotationSpace=BCS_BoneSpace
    alphaBoolBlend {blendOption=Linear}
    boneToModify {boneName=driver_frill_frt_r}
    rotation {yaw=-24.9993}
   }
   n5 = "Transform (Modify) Bone - Bone: Flap_Front_Parent_Space"(ComponentPose=@n4, Alpha=1, Translation="0,0,0", Rotation="0.000,0,-55", Scale="1,1,1") {
    rotationMode=BMM_Additive
    rotationSpace=BCS_BoneSpace
    alphaBoolBlend {blendOption=Linear}
    boneToModify {boneName=Flap_Front_Parent_Space}
    rotation {roll=-55}
   }
   n6 = Pose Driver - Source: upperarm_l!(SourcePose:Pose Link Structure=Blend Poses by bool(BlendPose_0=Component To Local(ComponentPose=@n2), BlendPose_1=Component To Local(ComponentPose=@n5), bActiveValue=Get bIsCrouching()) {blendType=Linear}) -> (Pose:Pose Link Structure) {
    is=title AnimGraphNode_PoseDriver
    poseAsset=&/F_MED_StrawberryPilot_ShoulderT4_5_PoseAsset
    onlyDriveBones [{boneName=dyn_shoulderBase_l}]
    poseTargets [
     {
      bApplyCustomCurve=false
      bIsHidden=false
      distanceMethod=DefaultMethod
      drivenName=base
      functionType=DefaultFunction
      targetScale=1
      boneTransforms [
       {targetRotation=<rot -37.9117 2.9184 0.9883>, targetTranslation=<9.8658 -1.2136 -0.4105>}
       {targetRotation=<rot -85.1913 -138.5624 143.0618>, targetTranslation=<5.4831 -0.1053 -3.2946>}
      ]
      customCurve {defaultValue=3.4028234663852886e+38, postInfinityExtrap=RCCE_Constant, preInfinityExtrap=RCCE_Constant, keys []}
      targetRotation=<rot 0 0 0>
     }
     {
      bApplyCustomCurve=false
      bIsHidden=false
      distanceMethod=DefaultMethod
      drivenName=pistolUp
      functionType=DefaultFunction
      targetScale=0.9554
      boneTransforms [
       {targetRotation=<rot 6.8358 41.9558 -103.2276>, targetTranslation=<9.8658 -1.2136 -0.4105>}
       {targetRotation=<rot -48.6577 16.7554 -6.8979>, targetTranslation=<5.4831 -0.1053 -3.2946>}
      ]
      customCurve {defaultValue=3.4028234663852886e+38, postInfinityExtrap=RCCE_Constant, preInfinityExtrap=RCCE_Constant, keys []}
      targetRotation=<rot 0 0 0>
     }
    ]
    rBFParams {radius=18}
    sourceBones [{boneName=upperarm_l}, {boneName=clavicle_l}]
   }
   n7 = Pose Driver - Source: upperarm_r!(SourcePose:Pose Link Structure=@n6) -> (Pose:Pose Link Structure) {
    is=title AnimGraphNode_PoseDriver
    poseAsset=&F_MED_StrawberryPilot_ShoulderT4_5_PoseAsset
    onlyDriveBones [{boneName=dyn_shoulderBase_r}]
    poseTargets [
     {
      bApplyCustomCurve=false
      bIsHidden=false
      distanceMethod=DefaultMethod
      drivenName=base
      functionType=DefaultFunction
      targetScale=1
      boneTransforms [
       {targetRotation=<rot -37.9117 2.9184 0.9883>, targetTranslation=<-9.8658 1.2136 0.4104>}
       {targetRotation=<rot -85.1914 41.4375 143.0618>, targetTranslation=<5.4831 -0.1053 3.2946>}
      ]
      customCurve {defaultValue=3.4028234663852886e+38, postInfinityExtrap=RCCE_Constant, preInfinityExtrap=RCCE_Constant, keys []}
      targetRotation=<rot 0 0 0>
     }
     {
      bApplyCustomCurve=false
      bIsHidden=false
      distanceMethod=DefaultMethod
      drivenName=pistolUp
      functionType=DefaultFunction
      targetScale=0.9894
      boneTransforms [
       {targetRotation=<rot 30.1907 88.7869 -82.525>, targetTranslation=<-9.8658 1.2136 0.4104>}
       {targetRotation=<rot -68.9927 -127.6323 -53.9075>, targetTranslation=<5.4831 -0.1053 3.2946>}
      ]
      customCurve {defaultValue=3.4028234663852886e+38, postInfinityExtrap=RCCE_Constant, preInfinityExtrap=RCCE_Constant, keys []}
      targetRotation=<rot 0 0 0>
     }
     {
      bApplyCustomCurve=false
      bIsHidden=false
      distanceMethod=DefaultMethod
      drivenName=rifleUp
      functionType=DefaultFunction
      targetScale=0.9894
      boneTransforms [
       {targetRotation=<rot -41.6104 102.6786 -60.7586>, targetTranslation=<-9.8658 1.2136 0.4104>}
       {targetRotation=<rot -76.9196 -140.3767 -39.1495>, targetTranslation=<5.4831 -0.1053 3.2946>}
      ]
      customCurve {defaultValue=3.4028234663852886e+38, postInfinityExtrap=RCCE_Constant, preInfinityExtrap=RCCE_Constant, keys []}
      targetRotation=<rot 0 0 0>
     }
    ]
    rBFParams {radius=37.1232}
    sourceBones [{boneName=upperarm_r}, {boneName=clavicle_r}]
   }
   Output Pose(Result=@n7) {layerGroup=None}
   n8 = NOT Boolean(A=Get bIsDBNO())
   n9 = Get JacketExternalForce() -> (JacketExternalForce:Vector)
   n10 = "Transform (Modify) Bone - Bone: dyn_thigh_scale_l"(
    ComponentPose=Local To Component(LocalPose=Copy Pose From Mesh() {bCopyCurves=true, bUseAttachedParent=true})
    Alpha=Get LegLeftSwingAlpha() -> (LegLeftSwingAlpha:float)
    Translation="-30,10,0"
    Rotation="0,0,0"
    Scale="7,7,7"
   ) {
    scaleMode=BMM_Replace
    scaleSpace=BCS_BoneSpace
    translationMode=BMM_Additive
    translationSpace=BCS_BoneSpace
    alphaBoolBlend {blendOption=Linear}
    boneToModify {boneName=dyn_thigh_scale_l}
    scale=<7 7 7>
    translation=<-30 10>
   }
   n11 = "Transform (Modify) Bone - Bone: dyn_thigh_scale_r"(ComponentPose=@n10, Alpha=Get LegRightSwingAlpha() -> (LegRightSwingAlpha:float), Translation="30,-13,0", Rotation="0,0,0", Scale="7,7,7") {
    scaleMode=BMM_Replace
    scaleSpace=BCS_BoneSpace
    translationMode=BMM_Additive
    translationSpace=BCS_BoneSpace
    alphaBoolBlend {blendOption=Linear}
    boneToModify {boneName=dyn_thigh_scale_r}
    scale=<7 7 7>
    translation=<30 -13>
   }
   n12 = "Constraint - dyn_collarFrontBaseDriven_r (PP)"!(ComponentPose=@n11, Alpha=1, ConstraintWeights_0=0.2, ConstraintWeights_1:float=0.8) -> (Pose) {
    alphaBoolBlend {blendOption=Linear}
    boneToModify {boneName=dyn_collarFrontBaseDriven_r}
    constraintSetup [{targetBone {boneName=spine_05}}, {targetBone {boneName=clavicle_r}}]
    constraintWeights [1,1]
   }
   n13 = "Constraint - dyn_collarFrontBaseDriven_l (PP)"!(ComponentPose=@n12, Alpha=1, ConstraintWeights_0=0.2, ConstraintWeights_1:float=0.8) -> (Pose) {
    alphaBoolBlend {blendOption=Linear}
    boneToModify {boneName=dyn_collarFrontBaseDriven_l}
    constraintSetup [{targetBone {boneName=spine_05}}, {targetBone {boneName=clavicle_l}}]
    constraintWeights [1,1]
   }
   n14 = "Constraint - dyn_collarMidBaseDriven_r (PP)"!(ComponentPose=@n13, Alpha=1, ConstraintWeights_0=0.6, ConstraintWeights_1:float=0.4) -> (Pose) {
    alphaBoolBlend {blendOption=Linear}
    boneToModify {boneName=dyn_collarMidBaseDriven_r}
    constraintSetup [{targetBone {boneName=spine_05}}, {targetBone {boneName=clavicle_r}}]
    constraintWeights [1,1]
   }
   n15 = "Constraint - dyn_collarMidBaseDriven_l (PP)"!(ComponentPose=@n14, Alpha=1, ConstraintWeights_0=0.6, ConstraintWeights_1:float=0.4) -> (Pose) {
    alphaBoolBlend {blendOption=Linear}
    boneToModify {boneName=dyn_collarMidBaseDriven_l}
    constraintSetup [{targetBone {boneName=spine_05}}, {targetBone {boneName=clavicle_l}}]
    constraintWeights [1,1]
   }
   n16 = "Constraint - dyn_frill_frt (PPPPP)"!(
    ComponentPose=@n15
    Alpha=1
    ConstraintWeights_0=0.1453
    ConstraintWeights_1:float=0.1503
    ConstraintWeights_2:float=0.3548
    ConstraintWeights_3:float=0.2236
    ConstraintWeights_4:float=0.126
   ) -> (Pose) {
    alphaBoolBlend {blendOption=Linear}
    boneToModify {boneName=dyn_frill_frt}
    constraintSetup [
     {targetBone {boneName=thigh_r}}
     {targetBone {boneName=pelvis}}
     {targetBone {boneName=thigh_twist_01_l}}
     {targetBone {boneName=thigh_twist_01_r}}
     {targetBone {boneName=thigh_l}}
    ]
    constraintWeights [1,1,1,1,1]
   }
   n17 = "Constraint - dyn_frill_bck (PPPPP)"!(
    ComponentPose=@n16
    Alpha=1
    ConstraintWeights_0=0.0484
    ConstraintWeights_1:float=0.4698
    ConstraintWeights_2:float=0.2345
    ConstraintWeights_3:float=0.2005
    ConstraintWeights_4:float=0.0468
   ) -> (Pose) {
    alphaBoolBlend {blendOption=Linear}
    boneToModify {boneName=dyn_frill_bck}
    constraintSetup [
     {targetBone {boneName=thigh_r}}
     {targetBone {boneName=pelvis}}
     {targetBone {boneName=thigh_twist_01_l}}
     {targetBone {boneName=thigh_twist_01_r}}
     {targetBone {boneName=thigh_l}}
    ]
    constraintWeights [1,1,1,1,1]
   }
   n18 = Anim Dynamics - Bone: dyn_jacketFront_r(ComponentPose=@n17, Alpha=1, ExternalForce=@n9) {
    angularDampingOverride=1
    angularSpringConstant=200
    bAngularSpring=true
    bGravityOverrideInSimSpace=true
    bOverrideAngularDamping=true
    simulationSpace=RootRelative
    alphaBoolBlend {blendOption=Linear}
    boundBone {boneName=dyn_jacketFront_r}
    componentAppliedLinearAccClamp=<2000 2000 3000>
    componentLinearAccScale=<0.8 0.8 0.8>
    componentLinearVelScale=<0.5 0.5 0.5>
    physicsBodyDefinitions [{
      sphereCollisionRadius=0
      boundBone {boneName=dyn_jacketFront_r}
      boxExtents=<15 3 15>
      constraintSetup {angularLimitsMax=<2 5 2>, angularLimitsMin=<-2 -5 -60>, angularTarget=<1 0 0>}
      localJointOffset=<10 0 0>
     }]
   }
   n19 = Anim Dynamics - Bone: dyn_jacketFront_l(ComponentPose=@n18, Alpha=1, ExternalForce=@n9) {
    angularDampingOverride=1
    angularSpringConstant=200
    bAngularSpring=true
    bGravityOverrideInSimSpace=true
    bOverrideAngularDamping=true
    simulationSpace=RootRelative
    alphaBoolBlend {blendOption=Linear}
    boundBone {boneName=dyn_jacketFront_l}
    componentAppliedLinearAccClamp=<2000 2000 3000>
    componentLinearAccScale=<0.8 0.8 0.8>
    componentLinearVelScale=<0.5 0.5 0.5>
    physicsBodyDefinitions [{
      sphereCollisionRadius=0
      boundBone {boneName=dyn_jacketFront_l}
      boxExtents=<15 3 15>
      constraintSetup {angularLimitsMax=<2 5 2>, angularLimitsMin=<-2 -5 -60>, angularTarget=<1 0 0>}
      localJointOffset=<-10 0 0>
     }]
   }
   n20 = Anim Dynamics - Bone: dyn_jacketSide_r(ComponentPose=@n19, Alpha=1, ExternalForce=@n9) {
    angularDampingOverride=1
    angularSpringConstant=200
    bAngularSpring=true
    bGravityOverrideInSimSpace=true
    bOverrideAngularDamping=true
    simulationSpace=RootRelative
    alphaBoolBlend {blendOption=Linear}
    boundBone {boneName=dyn_jacketSide_r}
    componentAppliedLinearAccClamp=<2000 2000 3000>
    componentLinearAccScale=<0.8 0.8 0.8>
    componentLinearVelScale=<0.5 0.5 0.5>
    physicsBodyDefinitions [{
      sphereCollisionRadius=0
      boundBone {boneName=dyn_jacketSide_r}
      boxExtents=<15 3 15>
      constraintSetup {angularLimitsMax=<2 5 2>, angularLimitsMin=<-2 -5 -60>, angularTarget=<1 0 0>}
      localJointOffset=<10 0 0>
     }]
   }
   n21 = Anim Dynamics - Bone: dyn_jacketSide_l(ComponentPose=@n20, Alpha=1, ExternalForce=@n9) {
    angularDampingOverride=1
    angularSpringConstant=200
    bAngularSpring=true
    bGravityOverrideInSimSpace=true
    bOverrideAngularDamping=true
    simulationSpace=RootRelative
    alphaBoolBlend {blendOption=Linear}
    boundBone {boneName=dyn_jacketSide_l}
    componentAppliedLinearAccClamp=<2000 2000 3000>
    componentLinearAccScale=<0.8 0.8 0.8>
    componentLinearVelScale=<0.5 0.5 0.5>
    physicsBodyDefinitions [{
      sphereCollisionRadius=0
      boundBone {boneName=dyn_jacketSide_l}
      boxExtents=<15 3 15>
      constraintSetup {angularLimitsMax=<2 5 2>, angularLimitsMin=<-2 -5 -60>, angularTarget=<1 0 0>}
      localJointOffset=<-10 0 0>
     }]
   }
   n22 = Anim Dynamics - Bone: dyn_jacketSideBack_r!(ComponentPose=@n21, bAlphaBoolEnabled:bool=@n8, ExternalForce=@n9) -> (Pose) {
    alphaInputType=Bool
    angularDampingOverride=1
    angularSpringConstant=4000
    bAngularSpring=true
    bGravityOverrideInSimSpace=true
    bOverrideAngularDamping=true
    simulationSpace=RootRelative
    alphaBoolBlend {blendOption=Linear}
    boundBone {boneName=dyn_jacketSideBack_r}
    componentAppliedLinearAccClamp=<2000 2000 3000>
    componentLinearAccScale=<0.75 0.75 0.75>
    componentLinearVelScale=<0.5 0.5 0.5>
    physicsBodyDefinitions [{
      sphereCollisionRadius=0
      boundBone {boneName=dyn_jacketSideBack_r}
      boxExtents=<15 3 15>
      constraintSetup {angularLimitsMax=<2 5 2>, angularLimitsMin=<-2 -5 -68>, angularTarget=<1 0 0>}
      localJointOffset=<10 0 0>
     }]
   }
   n23 = Anim Dynamics - Bone: dyn_jacketSideBack_l!(ComponentPose=@n22, bAlphaBoolEnabled:bool=@n8, ExternalForce=@n9) -> (Pose) {
    alphaInputType=Bool
    angularDampingOverride=1
    angularSpringConstant=4000
    bAngularSpring=true
    bGravityOverrideInSimSpace=true
    bOverrideAngularDamping=true
    simulationSpace=RootRelative
    alphaBoolBlend {blendOption=Linear}
    boundBone {boneName=dyn_jacketSideBack_l}
    componentAppliedLinearAccClamp=<2000 2000 3000>
    componentLinearAccScale=<0.75 0.75 0.75>
    componentLinearVelScale=<0.5 0.5 0.5>
    physicsBodyDefinitions [{
      sphereCollisionRadius=0
      boundBone {boneName=dyn_jacketSideBack_l}
      boxExtents=<15 3 15>
      constraintSetup {angularLimitsMax=<2 5 2>, angularLimitsMin=<-2 -5 -68>, angularTarget=<1 0 0>}
      localJointOffset=<-10 0 0>
     }]
   }
   n24 = Anim Dynamics - Bone: dyn_jacketBack!(ComponentPose=@n23, bAlphaBoolEnabled:bool=@n8, ExternalForce=@n9) -> (Pose) {
    alphaInputType=Bool
    angularDampingOverride=1
    angularSpringConstant=4000
    bAngularSpring=true
    bGravityOverrideInSimSpace=true
    bOverrideAngularDamping=true
    simulationSpace=RootRelative
    alphaBoolBlend {blendOption=Linear}
    boundBone {boneName=dyn_jacketBack}
    componentAppliedLinearAccClamp=<2000 2000 3000>
    componentLinearAccScale=<0.75 0.75 0.75>
    componentLinearVelScale=<0.5 0.5 0.5>
    physicsBodyDefinitions [{
      sphereCollisionRadius=0
      boundBone {boneName=dyn_jacketBack}
      boxExtents=<15 3 15>
      constraintSetup {angularLimitsMax=<2 5 2>, angularLimitsMin=<-2 -5 -68>, angularTarget=<1 0 0>}
      localJointOffset=<10 0 0>
     }]
   }
   n25 = Anim Dynamics - Bone: dyn_jacketZipper(ComponentPose=@n24, Alpha=1, ExternalForce="0,0,0") {
    angularDampingOverride=0.9999
    bAngularSpring=true
    bGravityOverrideInSimSpace=true
    bOverrideAngularDamping=true
    alphaBoolBlend {blendOption=Linear}
    boundBone {boneName=dyn_jacketZipper}
    componentAppliedLinearAccClamp=<2000 2000 2000>
    componentLinearAccScale=<0.4 0.4 0.7>
    componentLinearVelScale=<0.3 0.3 0.4>
    physicsBodyDefinitions [{
      sphereCollisionRadius=0
      boundBone {boneName=dyn_jacketZipper}
      boxExtents=<8 3 6>
      constraintSetup {angularLimitsMax=<0.5 0.5 45>, angularLimitsMin=<-0.5 -0.5 0>, angularTarget=<1 0 0>}
      localJointOffset=<4 0 0>
     }]
    relativeSpaceBone {boneName=dyn_jacketFront_l}
   }
   n26 = "Constraint - dyn_jumpsuitZipperBase (PP)"!(ComponentPose=@n25, Alpha=1, ConstraintWeights_0=0.25, ConstraintWeights_1:float=0.75) -> (Pose) {
    alphaBoolBlend {blendOption=Linear}
    boneToModify {boneName=dyn_jumpsuitZipperBase}
    constraintSetup [{targetBone {boneName=spine_05}}, {targetBone {boneName=spine_04}}]
    constraintWeights [1,1]
   }
   n27 = Anim Dynamics - Bone: dyn_jumpsuitZipper(ComponentPose=@n26, Alpha=1, ExternalForce="0,0,0") {
    angularDampingOverride=1
    angularSpringConstant=400
    bAngularSpring=true
    bGravityOverrideInSimSpace=true
    bOverrideAngularDamping=true
    simulationSpace=RootRelative
    alphaBoolBlend {blendOption=Linear}
    boundBone {boneName=dyn_jumpsuitZipper}
    componentAppliedLinearAccClamp=<2000 2000 2000>
    componentLinearAccScale=<0.4 0.4 0.7>
    componentLinearVelScale=<0.3 0.3 0.4>
    physicsBodyDefinitions [{
      sphereCollisionRadius=0
      boundBone {boneName=dyn_jumpsuitZipper}
      boxExtents=<8 3 6>
      constraintSetup {angularLimitsMax=<0.5 0.5 10>, angularLimitsMin=<-0.5 -0.5 -50>, angularTarget=<1 0 0>}
      localJointOffset=<3 0 0>
     }]
    relativeSpaceBone {boneName=dyn_jumpsuitZipperBase}
   }
   n28 = "Transform (Modify) Bone - Bone: dyn_jumpsuitZipper"(ComponentPose=@n27, Alpha="To Float (Boolean)"(InBool=Get bIsInFrontEnd()), Translation="0,0,0", Rotation="-10,0,-9", Scale="1,1,1") {rotationMode=BMM_Additive, alphaBoolBlend {blendOption=Linear}, boneToModify {boneName=dyn_jumpsuitZipper}, rotation {pitch=-10, roll=-9}}
   n29 = Get HeadSwingAlpha() -> (HeadSwingAlpha:float)
   n30 = "Transform (Modify) Bone - Bone: dyn_collarBack"(ComponentPose=@n28, Alpha=@n29, Translation="-3,10,0", Rotation="0,50,0", Scale="1,1,1") {
    rotationMode=BMM_Additive
    rotationSpace=BCS_BoneSpace
    alphaBoolBlend {blendOption=Linear}
    alphaScaleBiasClamp {bMapRange=true, inRange {min=0.4}, outRange {max=0.9}}
    boneToModify {boneName=dyn_collarBack}
    rotation {yaw=50}
    translation=<-3 10>
   }
   n31 = "Transform (Modify) Bone - Bone: dyn_collarSide_r"(ComponentPose=@n30, Alpha=@n29, Translation="-3,2,0", Rotation="0,45,0", Scale="1,1,1") {
    rotationMode=BMM_Additive
    rotationSpace=BCS_BoneSpace
    translationMode=BMM_Additive
    translationSpace=BCS_BoneSpace
    alphaBoolBlend {blendOption=Linear}
    alphaScaleBiasClamp {bMapRange=true, inRange {min=0.4}}
    boneToModify {boneName=dyn_collarSide_r}
    rotation {yaw=45}
    translation=<-3 2>
   }
   n32 = Get ArmRightSwingAlpha() -> (ArmRightSwingAlpha:float)
   n33 = "Transform (Modify) Bone - Bone: dyn_collarMidBase_r"(ComponentPose=@n31, Alpha=@n32, Translation="-0.75,-2,0", Rotation="0,0,0", Scale="1,1,1") {
    rotationMode=BMM_Replace
    rotationSpace=BCS_BoneSpace
    translationMode=BMM_Replace
    translationSpace=BCS_BoneSpace
    alphaBoolBlend {blendOption=Linear}
    alphaScaleBiasClamp {bMapRange=true, inRange {max=0.45}}
    boneToModify {boneName=dyn_collarMidBase_r}
    translation=<-0.75 -2>
   }
   n34 = "Transform (Modify) Bone - Bone: dyn_collarFrontBase_r"(ComponentPose=@n33, Alpha=@n32, Translation="-3,-2,3.5", Rotation="0,-30,50", Scale="1,1,1") {
    rotationMode=BMM_Replace
    rotationSpace=BCS_BoneSpace
    translationMode=BMM_Replace
    translationSpace=BCS_BoneSpace
    alphaBoolBlend {blendOption=Linear}
    alphaScaleBiasClamp {bMapRange=true, inRange {max=0.85}}
    boneToModify {boneName=dyn_collarFrontBase_r}
    rotation {roll=50, yaw=-30}
    translation=<-3 -2 3.5>
   }
   n35 = "Transform (Modify) Bone - Bone: upperarm_twist_01_push_r"(ComponentPose=@n34, Alpha=@n32, Translation="0,4,0", Rotation="0,0,0", Scale="1,0.2,1") {
    rotationSpace=BCS_BoneSpace
    scaleMode=BMM_Replace
    scaleSpace=BCS_BoneSpace
    translationMode=BMM_Additive
    translationSpace=BCS_BoneSpace
    alphaBoolBlend {blendOption=Linear}
    alphaScaleBiasClamp {bMapRange=true, inRange {max=0.4}}
    boneToModify {boneName=upperarm_twist_01_push_r}
    scale {y=0.2}
    translation {y=4}
   }
   n36 = Get ArmLeftSwingAlpha() -> (ArmLeftSwingAlpha:float)
   n37 = "Transform (Modify) Bone - Bone: dyn_collarMidBase_l"(ComponentPose=@n35, Alpha=@n36, Translation="0,22,0", Rotation="0,60,0", Scale="1,1,1") {
    rotationMode=BMM_Replace
    rotationSpace=BCS_BoneSpace
    translationMode=BMM_Replace
    translationSpace=BCS_BoneSpace
    alphaBoolBlend {blendOption=Linear}
    alphaScaleBiasClamp {inRange {min=0.5}}
    boneToModify {boneName=dyn_collarMidBase_l}
    rotation {yaw=60}
    translation {y=22}
   }
   n38 = "Transform (Modify) Bone - Bone: dyn_collarFrontBase_l"(ComponentPose=@n37, Alpha=@n36, Translation="0,1,3", Rotation="0,-30,62", Scale="1,1,1") {
    rotationMode=BMM_Replace
    rotationSpace=BCS_BoneSpace
    translationMode=BMM_Replace
    translationSpace=BCS_BoneSpace
    alphaBoolBlend {blendOption=Linear}
    alphaScaleBiasClamp {bMapRange=true, inRange {min=0.5}, outRange {max=0.48}}
    boneToModify {boneName=dyn_collarFrontBase_l}
    rotation {roll=62, yaw=-30}
    translation {y=1, z=3}
   }
   n39 = RigidBody(ComponentPose=@n38, Alpha=Get ArmRightSwingTargetingAlpha() -> (ArmRightSwingTargetingAlpha:float)) {
    bClampLinearTranslationLimitToRefPose=true
    bOverrideWorldGravity=true
    overridePhysicsAsset=&/F_MED_StrawberryPilot_ShoesCollar_PhysicsAsset
    simulationSpace=BaseBoneSpace
    alphaBoolBlend {blendOption=Linear}
    alphaScaleBiasClamp {bMapRange=true, inRange {min=0.7}, outRange {max=0.9}}
    baseBoneRef {boneName=pelvis}
    componentAppliedLinearAccClamp=<100 100 100>
    componentLinearAccScale=<0.45 0.45 0.45>
    componentLinearVelScale=<0.35 0.35 0.35>
    overrideWorldGravity {z=-980}
   }
   n40 = RigidBody(ComponentPose=@n39, Alpha=Get RigidBodyCrouchAlpha() -> (RigidBodyCrouchAlpha:float)) {
    bClampLinearTranslationLimitToRefPose=true
    bOverrideWorldGravity=true
    overridePhysicsAsset=&/F_MED_StrawberryPilot_T1_Body_Physics
    simulationSpace=BaseBoneSpace
    alphaBoolBlend {blendOption=Linear}
    baseBoneRef {boneName=root}
    componentAppliedLinearAccClamp=<1000 1000 1000>
    componentLinearAccScale=<1 1 1>
    componentLinearVelScale=<1 1 1>
    overrideWorldGravity {z=-980}
   }
   n41 = Get bIsDBNO()
   n42 = "Constraint - dyn_frill_bck (PP)"!(ComponentPose=@n40, bAlphaBoolEnabled:bool=@n41, ConstraintWeights_0=1, ConstraintWeights_1:float=1) -> (Pose) {
    alphaInputType=Bool
    alphaBoolBlend {blendOption=Linear}
    boneToModify {boneName=dyn_frill_bck}
    constraintSetup [{targetBone {boneName=dyn_frill_bck_l}}, {targetBone {boneName=dyn_frill_bck_r}}]
    constraintWeights [1,1]
   }
   n43 = "Transform (Modify) Bone - Bone: Flap_Front_Parent_Space"!(ComponentPose=@n42, bAlphaBoolEnabled:bool=@n41, Translation="0,0,0", Rotation="0,0,45", Scale="1,1,1") -> (Pose) {
    alphaInputType=Bool
    rotationMode=BMM_Additive
    rotationSpace=BCS_BoneSpace
    alphaBoolBlend {blendOption=Linear}
    boneToModify {boneName=Flap_Front_Parent_Space}
    rotation {roll=45}
   }
   Save cached pose 'PreCrouch'(Pose=Component To Local(ComponentPose=@n43)) {cacheName=PreCrouch}
  ]}
 EventGraph {nodes [
   Event BlueprintInitializeAnimation()
   > Delay(Duration=0.125)
   > ResetDynamics(InTeleportType=TeleportPhysics)
   > ResetDynamics(InTeleportType=ResetPhysics)
   Event OnBeginJumping()
   > RetriggerableDelay(Duration=0.25)
   > Set JacketExternalForce(JacketExternalForce:Vector="0,0,1250") -> (Output_Get:Vector)
   Event OnEndJumping()
   > RetriggerableDelay(Duration=0.375)
   > Set JacketExternalForce(JacketExternalForce="0, 0, 0")
   TryGetPawnOwner()
  ]}
]
variables [JacketExternalForce {type=Vector}]
