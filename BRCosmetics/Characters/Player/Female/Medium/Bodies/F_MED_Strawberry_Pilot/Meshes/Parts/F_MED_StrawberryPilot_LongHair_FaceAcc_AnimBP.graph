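# Post-processes long hair bones on the female Strawberry Pilot's face accessory mesh, adding head-swing correction and secondary motion.
# Copies the attached parent's pose, then constrains hair base bones to weighted spine, head and collar-area targets.
# HeadSwingDelta values, not declared here, drive corrective bone offsets and squashes, cached as PreDynamics.
# Five chained AnimDynamics nodes then spring front and back strands, using spine_05-driven planar limits against body clipping.
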AnimBlueprint /BRCosmetics/Characters/Player/Female/Medium/Bodies/F_MED_Strawberry_Pilot/Meshes/Parts/F_MED_StrawberryPilot_LongHair_FaceAcc_AnimBP
target_skeleton=/Game/Characters/Player/Female/Medium/Bodies/F_MED_Strawberry_Pilot/Meshes/Parts/F_MED_StrawberryPilot_LongHair_FaceAcc_Skeleton
graphs [
 AnimGraph {nodes [
   "Constraint - dyn_longHair_fr_01_base_r (PPPPPPPP)"!(
    ComponentPose
    Alpha=1
    ConstraintWeights_0=0.0408
    ConstraintWeights_1:float=0.5017
    ConstraintWeights_2:float=0.1639
    ConstraintWeights_3:float=0.0577
    ConstraintWeights_4:float=0.168
    ConstraintWeights_5:float=0.0477
    ConstraintWeights_6:float=0.0103
    ConstraintWeights_7:float=0.0065
   ) -> (Pose) {
    alphaBoolBlend {blendOption=Linear}
    boneToModify {boneName=dyn_longHair_fr_01_base_r}
    constraintSetup [
     {targetBone {boneName=spine_04}}
     {targetBone {boneName=spine_05}}
     {targetBone {boneName=clavicle_l}}
     {targetBone {boneName=dyn_collarMid_l}}
     {targetBone {boneName=dyn_collarFront_l}}
     {targetBone {boneName=dyn_collarSide_l}}
     {targetBone {boneName=dyn_collarBack}}
     {targetBone {boneName=upperarm_twist_01_l}}
    ]
    constraintWeights [1,1,1,1,1,1,1,1]
   }
   "Transform (Modify) Bone - Bone: dyn_longHair_fr_03_r"(Alpha=1, Translation="-1.734,-1.320,-3.199", Rotation="-12.882,-9.485,3.025", Scale="1,1,1") {
    rotationMode=BMM_Additive
    rotationSpace=BCS_BoneSpace
    scaleSpace=BCS_BoneSpace
    translationMode=BMM_Additive
    translationSpace=BCS_BoneSpace
    alphaBoolBlend {blendOption=Linear}
    boneToModify {boneName=dyn_longHair_fr_03_r}
    rotation=<rot -12.882 -9.485 3.025>
    translation=<-1.734 -1.32 -3.199>
   }
   "Transform (Modify) Bone"(Alpha=1, Translation="0,0,0", Rotation="0,0,0", Scale="1,1,1") {alphaBoolBlend {blendOption=Linear}, boneToModify=~}
   n1 = Anim Dynamics - Bone: dyn_longHair_fr_03_r(ComponentPose=Local To Component(LocalPose=Use cached pose 'PreDynamics'()), Alpha=1, ExternalForce="0,0,0") {
    angularDampingOverride=1
    angularSpringConstant=600
    bAngularSpring=true
    bGravityOverrideInSimSpace=true
    bOverrideAngularDamping=true
    simulationSpace=RootRelative
    alphaBoolBlend {blendOption=Linear}
    boundBone {boneName=dyn_longHair_fr_03_r}
    physicsBodyDefinitions [{
      collisionType=CustomSphere
      sphereCollisionRadius=2.5
      boundBone {boneName=dyn_longHair_fr_03_r}
      boxExtents=<15 15 8>
      constraintSetup {angularLimitsMax=<15 15 15>, angularLimitsMin=<-15 -15 -15>, angularTarget=<1 0 0>}
      localJointOffset=<2 1 1>
     }]
    planarLimits [
     {drivingBone {boneName=spine_05}, planeTransform {location=<9 0 0>, rotation=<rot 0 -60 90>}}
     {drivingBone {boneName=None}, planeTransform {location=<0 0 0>, rotation=<rot 0 0 0>}}
     {drivingBone {boneName=None}, planeTransform {location=<0 0 0>, rotation=<rot 0 0 0>}}
    ]
   }
   n2 = Anim Dynamics - Bone: dyn_longHair_fr_03_l(ComponentPose=@n1, Alpha=1, ExternalForce="0,0,0") {
    angularDampingOverride=1
    angularSpringConstant=600
    bAngularSpring=true
    bGravityOverrideInSimSpace=true
    bOverrideAngularDamping=true
    simulationSpace=RootRelative
    alphaBoolBlend {blendOption=Linear}
    boundBone {boneName=dyn_longHair_fr_03_l}
    physicsBodyDefinitions [{
      collisionType=CustomSphere
      sphereCollisionRadius=2.5
      boundBone {boneName=dyn_longHair_fr_03_l}
      boxExtents=<15 15 8>
      constraintSetup {angularLimitsMax=<15 15 15>, angularLimitsMin=<-15 -15 -15>, angularTarget=<1 0 0>}
      localJointOffset=<2 1 -0.25>
     }]
    planarLimits [
     {drivingBone {boneName=spine_05}, planeTransform {location=<9 0 0>, rotation=<rot 0 -60 90>}}
     {drivingBone {boneName=None}, planeTransform {location=<0 0 0>, rotation=<rot 0 0 0>}}
     {drivingBone {boneName=None}, planeTransform {location=<0 0 0>, rotation=<rot 0 0 0>}}
    ]
   }
   n3 = Anim Dynamics - Bone: dyn_hair_bk(ComponentPose=@n2, Alpha=1, ExternalForce="0,0,0") {
    angularDampingOverride=1
    angularSpringConstant=600
    bAngularSpring=true
    bGravityOverrideInSimSpace=true
    bOverrideAngularDamping=true
    simulationSpace=RootRelative
    alphaBoolBlend {blendOption=Linear}
    boundBone {boneName=dyn_hair_bk}
    componentAppliedLinearAccClamp=<2000 2000 2000>
    componentLinearAccScale=<0.6 0.6 0.6>
    componentLinearVelScale=<0.3 0.3 0.3>
    physicsBodyDefinitions [{
      collisionType=CustomSphere
      sphereCollisionRadius=2.1
      boundBone {boneName=dyn_hair_bk}
      boxExtents=<10 5 10>
      constraintSetup {angularLimitsMax=<15 15 15>, angularLimitsMin=<-15 -15 -15>, angularTarget=<1 0 0>}
      localJointOffset=<4 0 0>
     }]
    planarLimits [{drivingBone {boneName=spine_05}, planeTransform {location=<0 -2 0>, rotation=<rot 0 -16 90>}}]
   }
   n4 = Anim Dynamics - Bone: dyn_hair_bk_r(ComponentPose=@n3, Alpha=1, ExternalForce="0,0,0") {
    angularDampingOverride=1
    angularSpringConstant=600
    bAngularSpring=true
    bGravityOverrideInSimSpace=true
    bOverrideAngularDamping=true
    simulationSpace=RootRelative
    alphaBoolBlend {blendOption=Linear}
    boundBone {boneName=dyn_hair_bk_r}
    componentAppliedLinearAccClamp=<2000 2000 2000>
    componentLinearAccScale=<0.6 0.6 0.6>
    componentLinearVelScale=<0.3 0.3 0.3>
    physicsBodyDefinitions [{
      collisionType=CustomSphere
      sphereCollisionRadius=2
      boundBone {boneName=dyn_hair_bk_r}
      boxExtents=<8 4 8>
      constraintSetup {angularLimitsMax=<15 15 15>, angularLimitsMin=<-15 -15 -15>, angularTarget=<1 0 0>}
      localJointOffset=<4 0 0>
     }]
    planarLimits [{drivingBone {boneName=spine_05}, planeTransform {location=<0 0 0>, rotation=<rot 0 -16 89.9999>}}]
   }
   n5 = Anim Dynamics - Bone: dyn_hair_bk_l(ComponentPose=@n4, Alpha=1, ExternalForce="0,0,0") {
    angularDampingOverride=1
    angularSpringConstant=600
    bAngularSpring=true
    bGravityOverrideInSimSpace=true
    bOverrideAngularDamping=true
    simulationSpace=RootRelative
    alphaBoolBlend {blendOption=Linear}
    boundBone {boneName=dyn_hair_bk_l}
    componentAppliedLinearAccClamp=<2000 2000 2000>
    componentLinearAccScale=<0.6 0.6 0.6>
    componentLinearVelScale=<0.3 0.3 0.3>
    physicsBodyDefinitions [{
      collisionType=CustomSphere
      sphereCollisionRadius=2
      boundBone {boneName=dyn_hair_bk_l}
      boxExtents=<8 4 8>
      constraintSetup {angularLimitsMax=<15 15 15>, angularLimitsMin=<-15 -15 -15>, angularTarget=<1 0 0>}
      localJointOffset=<4 0 0>
     }]
    planarLimits [{drivingBone {boneName=spine_05}, planeTransform {location=<0 0 0>, rotation=<rot 0 -16.0001 89.9999>}}]
   }
   Output Pose(Result=Component To Local(ComponentPose=@n5)) {layerGroup=None}
   n6 = "Constraint - dyn_hair_bk_base_l (PP)"!(
    ComponentPose=Local To Component(LocalPose=Copy Pose From Mesh() {bUseAttachedParent=true})
    Alpha=1
    ConstraintWeights_0=0.25
    ConstraintWeights_1:float=0.75
   ) -> (Pose) {
    alphaBoolBlend {blendOption=Linear}
    boneToModify {boneName=dyn_hair_bk_base_l}
    constraintSetup [{targetBone {boneName=spine_05}}, {targetBone {boneName=head}}]
    constraintWeights [1,1]
   }
   n7 = "Constraint - dyn_hair_bk_base_r (PP)"!(ComponentPose=@n6, Alpha=1, ConstraintWeights_0=0.25, ConstraintWeights_1:float=0.75) -> (Pose) {
    alphaBoolBlend {blendOption=Linear}
    boneToModify {boneName=dyn_hair_bk_base_r}
    constraintSetup [{targetBone {boneName=spine_05}}, {targetBone {boneName=head}}]
    constraintWeights [1,1]
   }
   n8 = "Constraint - dyn_longHair_md_01_base_r (PP)"!(ComponentPose=@n7, Alpha=1, ConstraintWeights_0=0.5, ConstraintWeights_1:float=0.5) -> (Pose) {
    alphaBoolBlend {blendOption=Linear}
    boneToModify {boneName=dyn_longHair_md_01_base_r}
    constraintSetup [{targetBone {boneName=spine_05}}, {targetBone {boneName=head}}]
    constraintWeights [1,1]
   }
   n9 = "Constraint - dyn_longHair_fr_01_base_r (PP)"!(ComponentPose=@n8, Alpha=1, ConstraintWeights_0=0.25, ConstraintWeights_1:float=0.75) -> (Pose) {
    alphaBoolBlend {blendOption=Linear}
    boneToModify {boneName=dyn_longHair_fr_01_base_r}
    constraintSetup [{targetBone {boneName=spine_05}}, {targetBone {boneName=head}}]
    constraintWeights [1,1]
   }
   n10 = "Constraint - dyn_longHair_fr_01_base_l (PPPPPPPP)"!(
    ComponentPose=@n9
    Alpha=1
    ConstraintWeights_0=0.0408
    ConstraintWeights_1:float=0.5017
    ConstraintWeights_2:float=0.1639
    ConstraintWeights_3:float=0.0577
    ConstraintWeights_4:float=0.168
    ConstraintWeights_5:float=0.0477
    ConstraintWeights_6:float=0.0103
    ConstraintWeights_7:float=0.0065
   ) -> (Pose) {
    alphaBoolBlend {blendOption=Linear}
    boneToModify {boneName=dyn_longHair_fr_01_base_l}
    constraintSetup [
     {targetBone {boneName=spine_04}}
     {targetBone {boneName=spine_05}}
     {targetBone {boneName=clavicle_l}}
     {targetBone {boneName=dyn_collarMid_l}}
     {targetBone {boneName=dyn_collarFront_l}}
     {targetBone {boneName=dyn_collarSide_l}}
     {targetBone {boneName=dyn_collarBack}}
     {targetBone {boneName=upperarm_twist_01_l}}
    ]
    constraintWeights [1,1,1,1,1,1,1,1]
   }
   n11 = Get HeadSwingDeltaTargetLValue() -> (HeadSwingDeltaTargetLValue:float)
   n12 = "Transform (Modify) Bone - Bone: dyn_longHair_fr_02_l"(ComponentPose=@n10, Alpha=@n11, Translation="0.165,-0.4,2.75", Rotation="16.35,1.5,-21", Scale="1,1,0.75") {
    rotationMode=BMM_Additive
    rotationSpace=BCS_BoneSpace
    scaleMode=BMM_Replace
    scaleSpace=BCS_BoneSpace
    translationMode=BMM_Additive
    translationSpace=BCS_BoneSpace
    alphaBoolBlend {blendOption=Linear}
    alphaScaleBiasClamp {bMapRange=true, outRange {max=0.75}}
    boneToModify {boneName=dyn_longHair_fr_02_l}
    rotation=<rot 16.35 1.5 -21>
    scale {z=0.75}
    translation=<0.165 -0.4 2.75>
   }
   n13 = "Transform (Modify) Bone - Bone: dyn_longHair_fr_03_l"(ComponentPose=@n12, Alpha=@n11, Translation="-0.7,-1.3,-2.5", Rotation="-13.5,0.635,-0.2", Scale="1,1,1") {
    rotationMode=BMM_Additive
    rotationSpace=BCS_BoneSpace
    scaleSpace=BCS_BoneSpace
    translationMode=BMM_Additive
    translationSpace=BCS_BoneSpace
    alphaBoolBlend {blendOption=Linear}
    alphaScaleBiasClamp {bMapRange=true, outRange {max=0.75}}
    boneToModify {boneName=dyn_longHair_fr_03_l}
    rotation=<rot -13.5 0.635 -0.2>
    translation=<-0.7 -1.3 -2.5>
   }
   n14 = Get HeadSwingDeltaTargetRValue() -> (HeadSwingDeltaTargetRValue:float)
   n15 = "Transform (Modify) Bone - Bone: dyn_longHair_fr_01_r"(ComponentPose=@n13, Alpha=@n14, Translation="0.164,-0.380,-2.5", Rotation="16.341,1.490,-20.929", Scale="1,1,0.6") {
    rotationMode=BMM_Additive
    rotationSpace=BCS_BoneSpace
    scaleMode=BMM_Replace
    scaleSpace=BCS_BoneSpace
    translationMode=BMM_Additive
    translationSpace=BCS_BoneSpace
    alphaBoolBlend {blendOption=Linear}
    alphaScaleBiasClamp {bMapRange=true, clampMin=0.5}
    boneToModify {boneName=dyn_longHair_fr_01_r}
    rotation=<rot 16.341 1.49 -20.929>
    scale {z=0.6}
    translation=<0.164 -0.38 -2.5>
   }
   n16 = "Transform (Modify) Bone - Bone: dyn_longHair_fr_02_r"(ComponentPose=@n15, Alpha=@n14, Translation="-1.734,-1.320,-3.199", Rotation="-23,-10,5", Scale="1,1,1") {
    rotationMode=BMM_Additive
    rotationSpace=BCS_BoneSpace
    scaleSpace=BCS_BoneSpace
    translationMode=BMM_Additive
    translationSpace=BCS_BoneSpace
    alphaBoolBlend {blendOption=Linear}
    alphaScaleBiasClamp {bMapRange=true, clampMin=0.5, inRange {max=0.65}, outRange {min=0.25}}
    boneToModify {boneName=dyn_longHair_fr_02_r}
    rotation=<rot -23 -10 5>
    translation=<-1.734 -1.32 -3.199>
   }
   n17 = "Transform (Modify) Bone - Bone: dyn_longHair_md_01_base_r"(ComponentPose=@n16, Alpha=@n14, Translation="0,2,-2", Rotation="0,20,5", Scale="1,0.75,1") {
    rotationMode=BMM_Additive
    rotationSpace=BCS_BoneSpace
    scaleMode=BMM_Replace
    scaleSpace=BCS_BoneSpace
    translationMode=BMM_Additive
    translationSpace=BCS_BoneSpace
    alphaBoolBlend {blendOption=Linear}
    alphaScaleBiasClamp {bMapRange=true, clampMin=0.5, outRange {min=0.5}}
    boneToModify {boneName=dyn_longHair_md_01_base_r}
    rotation {roll=5, yaw=20}
    scale {y=0.75}
    translation {y=2, z=-2}
   }
   n18 = Get HeadSwingDeltaValue() -> (HeadSwingDeltaValue:float)
   n19 = "Transform (Modify) Bone - Bone: dyn_hair_bk_base_r"(ComponentPose=@n17, Alpha=@n18, Translation="0,3,0", Rotation="0,15,0.000", Scale="1,1,1") {
    rotationMode=BMM_Additive
    rotationSpace=BCS_BoneSpace
    scaleMode=BMM_Replace
    scaleSpace=BCS_BoneSpace
    translationMode=BMM_Additive
    translationSpace=BCS_BoneSpace
    alphaBoolBlend {blendOption=Linear}
    alphaScaleBiasClamp {bClampResult=true, clampMin=0.5}
    boneToModify {boneName=dyn_hair_bk_base_r}
    rotation {yaw=15}
    translation {y=3}
   }
   n20 = "Transform (Modify) Bone - Bone: dyn_hair_bk_base"(ComponentPose=@n19, Alpha=@n18, Translation="-0.000,2,0.000", Rotation="0.000,10,0.000", Scale="1,1,1") {
    rotationMode=BMM_Additive
    rotationSpace=BCS_BoneSpace
    scaleMode=BMM_Replace
    scaleSpace=BCS_BoneSpace
    translationMode=BMM_Additive
    translationSpace=BCS_BoneSpace
    alphaBoolBlend {blendOption=Linear}
    alphaScaleBiasClamp {bClampResult=true, clampMin=0.5}
    boneToModify {boneName=dyn_hair_bk_base}
    rotation {yaw=10}
    translation {y=2}
   }
   n21 = "Transform (Modify) Bone - Bone: dyn_hair_bk_base_l"(ComponentPose=@n20, Alpha=@n18, Translation="0,2,0", Rotation="0,10,0", Scale="1,1,1") {
    rotationMode=BMM_Additive
    rotationSpace=BCS_BoneSpace
    scaleMode=BMM_Replace
    scaleSpace=BCS_BoneSpace
    translationMode=BMM_Additive
    translationSpace=BCS_BoneSpace
    alphaBoolBlend {blendOption=Linear}
    alphaScaleBiasClamp {bClampResult=true, clampMin=0.5}
    boneToModify {boneName=dyn_hair_bk_base_l}
    rotation {yaw=10}
    translation {y=2}
   }
   n22 = "Transform (Modify) Bone - Bone: dyn_longHair_md_01_base_r"(ComponentPose=@n21, Alpha=@n18, Translation="0,1,-0.5", Rotation="0,10,0", Scale="1,0.6,1") {
    rotationMode=BMM_Additive
    rotationSpace=BCS_BoneSpace
    scaleMode=BMM_Replace
    scaleSpace=BCS_BoneSpace
    translationMode=BMM_Additive
    translationSpace=BCS_BoneSpace
    alphaBoolBlend {blendOption=Linear}
    alphaScaleBiasClamp {bClampResult=true, bMapRange=true, clampMin=0.5, inRange {max=0.5}, outRange {max=0.65}}
    boneToModify {boneName=dyn_longHair_md_01_base_r}
    rotation {yaw=10}
    scale {y=0.6}
    translation {y=1, z=-0.5}
   }
   Save cached pose 'PreDynamics'(Pose=Component To Local(ComponentPose=@n22)) {cacheName=PreDynamics}
  ]}
 EventGraph {links [], nodes []}
]
variables []
