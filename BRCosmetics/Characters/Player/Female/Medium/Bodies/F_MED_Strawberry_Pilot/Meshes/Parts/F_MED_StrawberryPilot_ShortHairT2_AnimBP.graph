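# Post-attach hair AnimBlueprint for the female Strawberry Pilot's ShortHairT2 part, adding secondary motion on the LongHair_FaceAcc skeleton.
# It copies the attached parent's pose, then additively offsets five hair base bones in bone space, weighted by HeadSwingDeltaValue.
# A series of root-relative AnimDynamics nodes then springs front, side and back hair bones, most held off spine and neck by spherical limits.
# The event graph is empty.
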AnimBlueprint /BRCosmetics/Characters/Player/Female/Medium/Bodies/F_MED_Strawberry_Pilot/Meshes/Parts/F_MED_StrawberryPilot_ShortHairT2_AnimBP
target_skeleton=/Game/Characters/Player/Female/Medium/Bodies/F_MED_Strawberry_Pilot/Meshes/Parts/F_MED_StrawberryPilot_LongHair_FaceAcc_Skeleton
graphs [
 AnimGraph {nodes [
   n1 = Get HeadSwingDeltaValue() -> (HeadSwingDeltaValue:float)
   n2 = "Transform (Modify) Bone - Bone: dyn_hair_bk_base_r"(
    ComponentPose=Local To Component(LocalPose=Copy Pose From Mesh() {bUseAttachedParent=true})
    Alpha=@n1
    Translation="0,2.5,0"
    Rotation="0,0,0"
    Scale="1,1,1"
   ) {
    rotationMode=BMM_Additive
    rotationSpace=BCS_BoneSpace
    translationMode=BMM_Additive
    translationSpace=BCS_BoneSpace
    alphaBoolBlend {blendOption=Linear}
    alphaScaleBiasClamp {bMapRange=true, inRange {max=0.85}}
    boneToModify {boneName=dyn_hair_bk_base_r}
    translation {y=2.5}
   }
   n3 = "Transform (Modify) Bone - Bone: dyn_hair_bk_base"(ComponentPose=@n2, Alpha=@n1, Translation="0,2.5,0", Rotation="0,0,0", Scale="1,1,1") {
    rotationMode=BMM_Additive
    rotationSpace=BCS_BoneSpace
    translationMode=BMM_Additive
    translationSpace=BCS_BoneSpace
    alphaBoolBlend {blendOption=Linear}
    alphaScaleBiasClamp {bMapRange=true, inRange {max=0.85}}
    boneToModify {boneName=dyn_hair_bk_base}
    translation {y=2.5}
   }
   n4 = "Transform (Modify) Bone - Bone: dyn_hair_bk_base_l"(ComponentPose=@n3, Alpha=@n1, Translation="0,2.5,0", Rotation="0,0,0", Scale="1,1,1") {
    rotationMode=BMM_Additive
    rotationSpace=BCS_BoneSpace
    translationMode=BMM_Additive
    translationSpace=BCS_BoneSpace
    alphaBoolBlend {blendOption=Linear}
    alphaScaleBiasClamp {bMapRange=true, inRange {max=0.85}}
    boneToModify {boneName=dyn_hair_bk_base_l}
    translation {y=2.5}
   }
   n5 = "Transform (Modify) Bone - Bone: dyn_hairSHort_base_r"(ComponentPose=@n4, Alpha=@n1, Translation="0,2,-3", Rotation="0,0,0", Scale="1,1,1") {
    rotationMode=BMM_Additive
    rotationSpace=BCS_BoneSpace
    translationMode=BMM_Additive
    translationSpace=BCS_BoneSpace
    alphaBoolBlend {blendOption=Linear}
    alphaScaleBiasClamp {bMapRange=true}
    boneToModify {boneName=dyn_hairSHort_base_r}
    translation {y=2, z=-3}
   }
   n6 = "Transform (Modify) Bone - Bone: dyn_longHair_md_01_base_r"(ComponentPose=@n5, Alpha=@n1, Translation="0,2,0", Rotation="0,0,0", Scale="1,1,1") {
    rotationMode=BMM_Additive
    rotationSpace=BCS_BoneSpace
    translationMode=BMM_Additive
    translationSpace=BCS_BoneSpace
    alphaBoolBlend {blendOption=Linear}
    alphaScaleBiasClamp {bMapRange=true}
    boneToModify {boneName=dyn_longHair_md_01_base_r}
    translation {y=2}
   }
   n7 = Anim Dynamics - Bone: dyn_shortHair_fr_01_l(ComponentPose=@n6, Alpha=1, ExternalForce="0,0,0") {
    angularDampingOverride=1
    angularSpringConstant=600
    bAngularSpring=true
    bGravityOverrideInSimSpace=true
    bOverrideAngularDamping=true
    bUseSphericalLimits=true
    simulationSpace=RootRelative
    alphaBoolBlend {blendOption=Linear}
    boundBone {boneName=dyn_shortHair_fr_01_l}
    componentAppliedLinearAccClamp=<2000 2000 2000>
    componentLinearAccScale=<0.6 0.6 0.6>
    componentLinearVelScale=<0.25 0.25 0.25>
    physicsBodyDefinitions [{
      collisionType=CustomSphere
      sphereCollisionRadius=2
      boundBone {boneName=dyn_shortHair_fr_01_l}
      boxExtents=<15 15 8>
      constraintSetup {angularLimitsMax=<10 10 10>, angularLimitsMin=<-10 -10 -10>, angularTarget=<1 0 0>}
      localJointOffset=<2 0 0>
     }]
    sphericalLimits [
     {limitRadius=7.75, drivingBone {boneName=spine_05}, sphereLocalOffset=<10 2 0>}
     {limitRadius=4.5, drivingBone {boneName=neck_02}, sphereLocalOffset=<0 0 0>}
    ]
   }
   n8 = Anim Dynamics - Bone: dyn_hairSHort_r(ComponentPose=@n7, Alpha=1, ExternalForce="0,0,0") {
    angularDampingOverride=1
    angularSpringConstant=600
    bAngularSpring=true
    bGravityOverrideInSimSpace=true
    bOverrideAngularDamping=true
    bUseSphericalLimits=true
    simulationSpace=RootRelative
    alphaBoolBlend {blendOption=Linear}
    boundBone {boneName=dyn_hairSHort_r}
    componentAppliedLinearAccClamp=<2000 2000 2000>
    componentLinearAccScale=<0.6 0.6 0.6>
    componentLinearVelScale=<0.25 0.25 0.25>
    physicsBodyDefinitions [{
      collisionType=CustomSphere
      sphereCollisionRadius=2
      boundBone {boneName=dyn_hairSHort_r}
      boxExtents=<15 15 8>
      constraintSetup {angularLimitsMax=<25 25 10>, angularLimitsMin=<-25 -25 -10>, angularTarget=<1 0 0>}
      localJointOffset=<2 0 0>
     }]
    sphericalLimits [{limitRadius=9, drivingBone {boneName=spine_05}, sphereLocalOffset=<10 2 0>}]
   }
   n9 = Anim Dynamics - Bone: dyn_shortHair_fr_r(ComponentPose=@n8, Alpha=1, ExternalForce="0,0,0") {
    angularDampingOverride=1
    angularSpringConstant=600
    bAngularSpring=true
    bGravityOverrideInSimSpace=true
    bOverrideAngularDamping=true
    bUseSphericalLimits=true
    simulationSpace=RootRelative
    alphaBoolBlend {blendOption=Linear}
    boundBone {boneName=dyn_shortHair_fr_r}
    componentAppliedLinearAccClamp=<2000 2000 2000>
    componentLinearAccScale=<0.6 0.6 0.6>
    componentLinearVelScale=<0.25 0.25 0.25>
    physicsBodyDefinitions [{
      collisionType=CustomSphere
      sphereCollisionRadius=2
      boundBone {boneName=dyn_shortHair_fr_r}
      boxExtents=<15 15 8>
      constraintSetup {angularLimitsMax=<25 25 10>, angularLimitsMin=<-25 -25 -10>, angularTarget=<1 0 0>}
      localJointOffset=<2 0 0>
     }]
    sphericalLimits [{limitRadius=9, drivingBone {boneName=spine_05}, sphereLocalOffset=<10 2 0>}]
   }
   n10 = Anim Dynamics - Bone: dyn_longHair_md_01_r(ComponentPose=@n9, Alpha=1, ExternalForce="0,0,0") {
    angularDampingOverride=1
    angularSpringConstant=600
    bAngularSpring=true
    bGravityOverrideInSimSpace=true
    bOverrideAngularDamping=true
    bUseSphericalLimits=true
    simulationSpace=RootRelative
    alphaBoolBlend {blendOption=Linear}
    boundBone {boneName=dyn_longHair_md_01_r}
    componentAppliedLinearAccClamp=<2000 2000 2000>
    componentLinearAccScale=<0.6 0.6 0.6>
    componentLinearVelScale=<0.25 0.25 0.25>
    physicsBodyDefinitions [{
      collisionType=CustomSphere
      sphereCollisionRadius=2
      boundBone {boneName=dyn_longHair_md_01_r}
      boxExtents=<15 15 8>
      constraintSetup {angularLimitsMax=<3 3 25>, angularLimitsMin=<-3 -3 -25>, angularTarget=<1 0 0>}
      localJointOffset=<5 1 0>
     }]
   }
   n11 = Anim Dynamics - Bone: dyn_hair_bk(ComponentPose=@n10, Alpha=1, ExternalForce="0,0,0") {
    angularDampingOverride=1
    angularSpringConstant=600
    bAngularSpring=true
    bGravityOverrideInSimSpace=true
    bOverrideAngularDamping=true
    bUsePlanarLimit=false
    bUseSphericalLimits=true
    simulationSpace=RootRelative
    alphaBoolBlend {blendOption=Linear}
    boundBone {boneName=dyn_hair_bk}
    componentAppliedLinearAccClamp=<2000 2000 2000>
    componentLinearAccScale=<1 1 1>
    componentLinearVelScale=<0.4 0.4 0.4>
    physicsBodyDefinitions [{
      collisionType=CustomSphere
      sphereCollisionRadius=2.1
      boundBone {boneName=dyn_hair_bk}
      boxExtents=<12 7 12>
      constraintSetup {angularLimitsMax=<15 25 15>, angularLimitsMin=<-15 -25 -15>, angularTarget=<1 0 0>}
      localJointOffset=<6 0 0>
     }]
    planarLimits [{drivingBone {boneName=spine_05}, planeTransform {location=<0 -3 0>, rotation=<rot 0 -16.0001 89.9999>}}]
    sphericalLimits [{limitRadius=21.5, drivingBone {boneName=spine_05}, sphereLocalOffset=<-2 7 0>}]
   }
   n12 = Anim Dynamics - Bone: dyn_hair_bk_r(ComponentPose=@n11, Alpha=1, ExternalForce="0,0,0") {
    angularDampingOverride=1
    angularSpringConstant=400
    bAngularSpring=true
    bGravityOverrideInSimSpace=true
    bOverrideAngularDamping=true
    bUsePlanarLimit=false
    bUseSphericalLimits=true
    simulationSpace=RootRelative
    alphaBoolBlend {blendOption=Linear}
    boundBone {boneName=dyn_hair_bk_r}
    componentAppliedLinearAccClamp=<2000 2000 2000>
    componentLinearAccScale=<1 1 1>
    componentLinearVelScale=<0.4 0.4 0.4>
    physicsBodyDefinitions [{
      collisionType=CustomSphere
      sphereCollisionRadius=2
      boundBone {boneName=dyn_hair_bk_r}
      boxExtents=<12 7 12>
      constraintSetup {angularLimitsMax=<15 25 15>, angularLimitsMin=<-15 -25 -15>, angularTarget=<1 0 0>}
      localJointOffset=<6 0 0>
     }]
    planarLimits [{drivingBone {boneName=spine_05}, planeTransform {location=<0 -3 0>, rotation=<rot 0 -16 89.9999>}}]
    sphericalLimits [{limitRadius=21.5, drivingBone {boneName=spine_05}, sphereLocalOffset=<-2 7 0>}]
   }
   n13 = Anim Dynamics - Bone: dyn_hair_bk_l(ComponentPose=@n12, Alpha=1, ExternalForce="0,0,0") {
    angularDampingOverride=1
    angularSpringConstant=400
    bAngularSpring=true
    bGravityOverrideInSimSpace=true
    bOverrideAngularDamping=true
    bUsePlanarLimit=false
    bUseSphericalLimits=true
    simulationSpace=RootRelative
    alphaBoolBlend {blendOption=Linear}
    boundBone {boneName=dyn_hair_bk_l}
    componentAppliedLinearAccClamp=<2000 2000 2000>
    componentLinearAccScale=<1 1 1>
    componentLinearVelScale=<0.4 0.4 0.4>
    physicsBodyDefinitions [{
      collisionType=CustomSphere
      sphereCollisionRadius=2
      boundBone {boneName=dyn_hair_bk_l}
      boxExtents=<12 7 12>
      constraintSetup {angularLimitsMax=<15 25 15>, angularLimitsMin=<-15 -25 -15>, angularTarget=<1 0 0>}
      localJointOffset=<6 0 0>
     }]
    planarLimits [{drivingBone {boneName=spine_05}, planeTransform {location=<0 -3 0>, rotation=<rot 0 -16.0002 90>}}]
    sphericalLimits [{limitRadius=21.5, drivingBone {boneName=spine_05}, sphereLocalOffset=<-2 7 0>}]
   }
   Output Pose(Result=Component To Local(ComponentPose=@n13)) {layerGroup=None}
  ]}
 EventGraph {links [], nodes []}
]
variables []
